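# Sound cue for a single Origin voice-over line from Battle Pass Season 20 Week 2, apparently quest dialogue.
# A lone Dialogue Player root plays the dialogue wave in a context with The Origin as speaker and no targets.
# The cue routes through the narrator VO sound class, uses the shared Season 20 quest attenuation, and boosts volume.
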SoundCue /BattlePassS20/Audio/VO/Cues/Week_2/01/ORIGIN_2022_REGS_Week02_01_02_010_Dialogue_Cue
root=DialoguePlayer_0
nodes [{
  id=DialoguePlayer_0
  properties {dialogueWaveParameter {
    dialogueWave=/BattlePassS20/Audio/VO/Waves/Origin/ORIGIN_2022_REGS_Week02_01_02_010_Dialogue
    context {speaker=/BattlePassS20/Audio/VO/TheOrigin_DV, targets []}
   }}
 }]
properties {
 attenuationSettings=/BattlePassS20/Audio/VO/S20_Quest_Att
 maxDistance=2097152
 soundClassObject=/Game/Sounds/FortSoundClassesAndModes/Fort_VO_Narrator
 volumeMultiplier=2
}
